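# Level Sequence for Section 01 of PistolCanary, assembling that section's sub-sequences and world data layer timing.
# Sub-sequence tracks run "mask maker hit" into "comic heroes intro", with "comic intro v2" overlapping the handoff.
# Data Layer tracks cover Day, S1, Castle and S2 layers, with Castle and S2 starting later and S2 appearing twice identically.
# A visibility track on the possessed generic game countdown Blueprint defaults it to hidden.
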LevelSequence /PistolCanary/Sequence/PistolCanary_Section_01/PistolCanary_Section_01
bindings [PistolCanary_BP_Countdown {
  id=#1
  possessed_class=BP_Generic_Game_Countdown_C
  tracks [{
    type=MovieSceneVisibilityTrack
    sections [{type=MovieSceneVisibilitySection, channels [None {type=MovieSceneScriptingBoolChannel, default=false}]}]
   }]
 }]
display_rate {denominator=1, numerator=60}
tick_resolution {denominator=1, numerator=24000}
tracks [
 {type=MovieSceneSubTrack, display_name=mmh - mask maker hit, sections [{type=MovieSceneSubSection, end_frame=8748, start_frame=3600}]}
 {type=MovieSceneSubTrack, display_name=chi - comic heroes intro, sections [{type=MovieSceneSubSection, end_frame=9413, start_frame=8748}]}
 {type=MovieSceneDataLayerTrack, display_name=Data Layer - S2, sections [{type=MovieSceneDataLayerSection, end_frame=9413, start_frame=8868}]}
 {
  type=MovieSceneDataLayerTrack
  display_name=Data Layer - Day3_to_EndS0
  sections [{type=MovieSceneDataLayerSection, end_frame=9413, start_frame=3600}]
 }
 {
  type=MovieSceneDataLayerTrack
  display_name=Data Layer - S1_to_End_S3
  sections [{type=MovieSceneDataLayerSection, end_frame=9413, start_frame=3600}]
 }
 {type=MovieSceneDataLayerTrack, display_name=Data Layer - S1, sections [{type=MovieSceneDataLayerSection, end_frame=9413, start_frame=3600}]}
 {
  type=MovieSceneDataLayerTrack
  display_name=Data Layer - Castle
  sections [{type=MovieSceneDataLayerSection, end_frame=9413, start_frame=6529}]
 }
 {
  type=MovieSceneDataLayerTrack
  display_name=Data Layer - Day1_to_EndS0
  sections [{type=MovieSceneDataLayerSection, end_frame=9413, start_frame=3600}]
 }
 {type=MovieSceneDataLayerTrack, display_name=Data Layer - S2, sections [{type=MovieSceneDataLayerSection, end_frame=9413, start_frame=8868}]}
 {type=MovieSceneSubTrack, display_name=comic intro v2, sections [{type=MovieSceneSubSection, end_frame=9161, start_frame=8348}]}
]
